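# Physics asset for the M_MED_VigorKnackDill body mesh, a medium male player character cosmetic in BRCosmetics.
# Its constraint list is empty, so it defines no joints for ragdoll or secondary physics motion.
# No body setups or collision shapes are shown.
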PhysicsAsset /BRCosmetics/Characters/Player/Male/Medium/Bodies/M_MED_VigorKnackDill/Meshes/M_MED_VigorKnackDill_PhysicsAsset
constraints []
